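# Additive, unlit, two-sided particle sprite material that draws a rotating, flickering halo corona flare.
# Two counter-rotated spikeball texture samples combine with a radial gradient lookup, brightness and particle color, with dynamic parameters driving size, spin, sine-based flicker and color jitter.
# Static switches allow gradient desaturation and hue shift after HDR-from-LDR remapping.
# Separate branches compute polar sine rays and particle-alpha depth/near-camera fade, bypassing the fade on mobile.
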Material /Figment_S10/Effects/Fort_Effects/Materials/Flares/M_Flare_Halo_Corona
graph {
 properties {
  bUsedWithParticleSprites=true
  blendMode=BLEND_Additive
  pixelDepthOffsetMode=PDOM_Legacy
  shadingModel=MSM_Unlit
  twoSided=true
  usedShadingModels=Unlit
 }
 nodes [
  n1 = Multiply(A=Time(), B=ScalarParameter() {defaultValue=0.5, parameterName=RotateSpeed})
  n2 = Add(A=@n1, B=@n1) {constB=0.2}
  n3 = MaterialFunctionCall(Float2=Subtract(A=TextureCoordinate() {uTiling=2, vTiling=2})) {materialFunction=&BreakOutFloat2Components}
  n4 = Divide(A=ComponentMask(None=Arctangent2Fast(Y=@n3.R, X=@n3.R)) {r=true}, B=Constant() {r=6.2832})
  n5 = Sine(None=Add(A=Multiply(A=@n2) {constB=-0.03}, B=@n4)) {period=0.2}
  n6 = Sine(None=Add(A=Multiply(A=@n1) {constB=0.04}, B=@n4)) {period=0.2}
  n7 = Sine(None=Add(A=Multiply(A=@n1) {constB=-0.02}, B=@n4)) {period=0.02}
  AppendVector(A=AppendVector(A=@n5, B=@n6), B=@n7)
  Constant() {r=6.2832}
  n8 = ParticleColor()
  n9 = DepthFade(
   Opacity=MaterialFunctionCall(Distance=ScalarParameter() {defaultValue=200, parameterName=NearCamFade}) {materialFunction=&Near_Camera_Fade}.Result
   FadeDistance=ScalarParameter() {defaultValue=100, parameterName=DepthFade}
  )
  n10 = Constant() {r=1}
  Multiply(A=@n8.A, B=FeatureLevelSwitch(Default=@n9, ES2=@n10, ES3_1=@n10, SM4_REMOVED=@n9))
  n11 = Multiply(A=@n6) {constB=0.5}
  n12 = Add(A=Add(A=Multiply(A=@n5) {constB=0.4}, B=Multiply(A=@n5, B=@n11)), B=Multiply(A=@n6, B=Multiply(A=@n7) {constB=0.2}))
  n13 = Constant() {r=0.5}
  n14 = SphereMask(A=TextureCoordinate(), B=@n13, Radius=@n13) {hardnessPercent=0}
  n15 = Add(
   A=Power(Base=ConstantBiasScale(None=@n12), Exp=LinearInterpolate(Alpha=@n14) {constA=5}) {constExponent=3}
   B=LinearInterpolate(A=Constant3Vector() {constant=<rgba 0.125 0.182 0.25 0>}, B=Constant3Vector() {constant=<rgba 0.8 1 0.9573 0>}, Alpha=@n14)
  )
  Saturate(None=Power(Base=Multiply(A=@n15, B=@n14)))
  n16 = TextureCoordinate()
  Sine(None=@n16)
  Sine(None=@n16)
  n17 = DynamicParameter() {defaultValue=<rgba 0.6 1 1 1>, paramNames [Size,RotateSpeed,Flicker,ColorJitter]}
  n18 = Multiply(A=Time())
  n19 = Multiply(A=Sine(None=@n18) {period=0.155}, B=Sine(None=@n18) {period=0.2})
  n20 = Sine(None=@n18) {period=0.43}
  n21 = Clamp(None=Multiply(A=Sine(None=@n18) {period=0.075}, B=Multiply(A=@n19, B=@n20))) {maxDefault=3}
  n22 = Add(A=TextureCoordinate() {uTiling=0.25, vTiling=0.25}) {constB=0.375}
  n23 = Multiply(A=Time(), B=@n17.RotateSpeed)
  n24 = TextureCoordinate() {uTiling=0.15, vTiling=0.15}
  n25 = TextureSample(UVs=Rotator(Coordinate=Add(A=@n24) {constB=0.425}, Time=@n23) {speed=-0.72}) {texture=/Game/Effects/Fort_Effects/Textures/Flares/T_Flare_Spikeball}
  n26 = Multiply(
   A=TextureSample(UVs=Rotator(Coordinate=@n22, Time=@n23)) {texture=&T_Flare_Spikeball}.RGB
   B=Power(Base=OneMinus(None=@n25.RGB)) {constExponent=5}
  )
  n27 = Multiply(A=ScalarParameter() {defaultValue=1, parameterName=Brightness}) {constB=7}
  n28 = Constant() {r=0.5}
  n29 = TextureSample(UVs=SphereMask(A=TextureCoordinate(), B=@n28, Radius=@n28, Hardness=LinearInterpolate(Alpha=@n17.Size) {constA=1, constB=0}) {hardnessPercent=25}) {texture=/Game/Effects/Fort_Effects/Textures/Grad/T_Halo_Corona_Gradient}
  n30 = Desaturation(None=@n29.RGB)
  n31 = Multiply(
   A=Multiply(A=LinearInterpolate(A=@n21, Alpha=@n17.Flicker) {constA=1}, B=@n8.RGB)
   B=Multiply(A=Multiply(A=@n26, B=@n27) {constB=80}, B=StaticSwitchParameter(True=@n30, False=@n29.RGB) {parameterName=UseParticleColor})
  )
  n32 = MaterialFunctionCall(LDR Input=Clamp(None=@n31), HDR Tint=@n31) {materialFunction=&DeriveHDRfromLDR}
  StaticSwitchParameter(
   True=MaterialFunctionCall(Hue Shift Percentage=@n17.ColorJitter, Texture=@n32.RemappedRange) {materialFunction=&HueShift}.Result
   False=@n32.RemappedRange
  ) {parameterName=UseColorJitter?}
 ]
}
